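# Translucent, two-sided unlit trail material for the Sparkle Remix glider, usable on beam trails and Niagara ribbons.
# Three panned disco-tile masks with sine-animated centres blend into sparkles, optionally added to scrolling clouds to mix two stripe colours, tinted by particle colour and time of day.
# Opacity combines sparkles, particle alpha, UV edge masks, fresnel and near-camera fade, plus depth fade off mobile.
# Pattern-cycling pulse and world-position branches appear unconnected.
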Material /Game/Gadgets/Assets/VinderTech_GliderChute/Glider_Sparkle_Remix/FX/M_Glider_Sparkle_Remix_Trail
graph {
 properties {
  bUsedWithBeamTrails=true
  bUsedWithNiagaraRibbons=true
  blendMode=BLEND_Translucent
  pixelDepthOffsetMode=PDOM_Legacy
  shadingModel=MSM_Unlit
  translucencyLightingMode=TLM_VolumetricPerVertexNonDirectional
  twoSided=true
  usedShadingModels=Unlit
 }
 nodes [
  n1 = ObjectPositionWS()
  n2 = VectorParameter() {parameterName=InitialOffset, defaultValue=<rgba 0 0 -1000 1>}
  Add(A=SphereMask(A=WorldPosition().XYZ, B=Add(A=@n1, B=@n2.RGB), Radius=Constant() {r=2000}, Hardness=Constant()), B=Constant() {r=-200}) {constB=-20}
  n3 = ActorPositionWS()
  n4 = ComponentMask(None=Subtract(A=WorldPosition().XYZ, B=Multiply(A=@n3) {constB=5})) {g=true, r=true}
  Divide(A=@n4) {constB=128}
  Divide(A=@n4) {constB=1720.3199}
  n5 = Multiply(A=ComponentMask(None=TextureCoordinate()) {g=true, r=true}, B=ScalarParameter() {defaultValue=0.5, group=EQ Shape, parameterName=X_Nr_Bands}) {constB=10}
  n6 = ScalarParameter() {defaultValue=0.25, parameterName="Rotation Angle (0-1) (S)"}
  n7 = TextureSample(UVs=Panner(Coordinate=MaterialFunctionCall(UVs=@n5, "Rotation Angle (0-1)"=@n6) {materialFunction=&CustomRotator}.Rotated Values) {speedY=-0.8}) {samplerType=SAMPLERTYPE_LinearColor, texture=/Game/Weapons/WeaponSkins/Textures/DiscoTiles1_M}
  n8 = MaterialFunctionCall(Speed=ScalarParameter() {defaultValue=0.2, parameterName="Speed2 (S)"}) {materialFunction=&DebugTimeSine}
  n9 = MaterialFunctionCall(Speed=ScalarParameter() {defaultValue=0.15, parameterName="Speed (S)"}) {materialFunction=&DebugTimeSine}
  n10 = Constant() {r=0.5}
  n11 = Constant() {r=1}
  n12 = ComponentMask(None=TextureCoordinate()) {g=true, r=true}
  n13 = ScalarParameter() {defaultValue=0.5, group=EQ Shape, parameterName=Y_Nr_Bands}
  n14 = OneMinus(None=Frac(None=Multiply(A=@n12, B=@n13) {constB=10}))
  n15 = Multiply(A=TextureSample(UVs=Panner(Coordinate=@n14) {speedX=0.5}) {samplerType=SAMPLERTYPE_LinearColor, texture=&DiscoTiles1_M}.RGB) {constB=0.5}
  n16 = MaterialFunctionCall(Speed=ScalarParameter() {defaultValue=0.2, parameterName="Speed2 (S)"}) {materialFunction=&DebugTimeSine}
  n17 = MaterialFunctionCall(Speed=ScalarParameter() {defaultValue=0.4, parameterName="Speed3 (S)"}) {materialFunction=&DebugTimeSine}
  n18 = Constant() {r=0.5}
  n19 = Constant() {r=1}
  n20 = ComponentMask(None=TextureCoordinate()) {g=true, r=true}
  n21 = ScalarParameter() {defaultValue=1, group=EQ Shape, parameterName=Z_Nr_Bands}
  n22 = OneMinus(None=Frac(None=Multiply(A=@n20, B=@n21) {constB=10}))
  n23 = Multiply(A=TextureSample(UVs=Panner(Coordinate=@n22) {speedX=2}) {samplerType=SAMPLERTYPE_LinearColor, texture=&DiscoTiles1_M}.RGB) {constB=0.5}
  n24 = MaterialFunctionCall(Speed=ScalarParameter() {defaultValue=0.2, parameterName="Speed2 (S)"}) {materialFunction=&DebugTimeSine}
  n25 = MaterialFunctionCall(Speed=ScalarParameter() {defaultValue=0.4, parameterName="Speed3 (S)"}) {materialFunction=&DebugTimeSine}
  n26 = Constant() {r=0.5}
  n27 = Constant() {r=1}
  n28 = LinearInterpolate(
   A=SphereMask(
    A=ComponentMask(None=@n7.RGB) {g=true, r=true}
    B=AppendVector(A=@n8.Result, B=@n9.Result)
    Radius=LinearInterpolate(A=@n10, B=@n11, Alpha=@n8.Result)
    Hardness=ScalarParameter() {defaultValue=0.25, parameterName=Hardness}
   )
   B=SphereMask(
    A=ComponentMask(None=@n15) {g=true, r=true}
    B=AppendVector(A=@n16.Result, B=@n17.Result)
    Radius=LinearInterpolate(A=@n18, B=@n19, Alpha=@n16.Result)
    Hardness=ScalarParameter() {defaultValue=0.25, parameterName=Hardness}
   )
   Alpha=SphereMask(
    A=ComponentMask(None=@n23) {g=true, r=true}
    B=AppendVector(A=@n24.Result, B=@n25.Result)
    Radius=LinearInterpolate(A=@n26, B=@n27, Alpha=@n24.Result)
    Hardness=ScalarParameter() {defaultValue=0.25, parameterName=Hardness}
   )
  )
  n29 = Clamp(None=@n28)
  n30 = TextureSample(UVs=Panner(Coordinate=TextureCoordinate() {uTiling=0, vTiling=0.3}) {speedX=-0.2, speedY=-0.2}) {texture=/Game/Effects/Fort_Effects/PlayerHealthDamage/T_DBNO_Cloud}
  n31 = TextureSample(UVs=Panner(Coordinate=TextureCoordinate() {uTiling=0, vTiling=0.2}) {speedX=-0.25, speedY=-0.2}) {texture=&T_DBNO_Cloud}
  n32 = TextureCoordinate() {uTiling=0.2}
  n33 = ConstantBiasScale(None=TextureSample(UVs=Panner(Coordinate=@n32) {bFractionalPart=true, speedX=-0.65}) {texture=&T_NoisePack_01}.R) {bias=-0.5, scale=0.2}
  n34 = ComponentMask(None=TextureCoordinate()) {r=true}
  n35 = OneMinus(None=@n34)
  n36 = Multiply(A=ComponentMask(None=Add(A=@n30.RGB, B=@n31.RGB)) {g=true}, B=LinearInterpolate(Alpha=Add(A=@n33, B=@n35)) {constA=0.01, constB=0.9})
  n37 = Add(A=@n29, B=@n36)
  n38 = ParticleColor()
  n39 = Multiply(
   A=LinearInterpolate(
    A=VectorParameter() {parameterName=Color_Stripes_Top_1, defaultValue=<rgba 1 0.019 0.804 1>}.RGB
    B=VectorParameter() {parameterName=Color_Stripes_Top_2, defaultValue=<rgba 0.5063 0.0724 1 1>}.RGB
    Alpha=StaticSwitchParameter(True=@n37, False=@n36) {parameterName=UseSparkles?}
   )
   B=@n38.RGB
  )
  MaterialFunctionCall(Emissive=MaterialFunctionCall(Color_Morning=@n39, Color_Day=@n39, Color_Evening=@n39, Color_Night=Multiply(A=@n39) {constB=0.25}) {materialFunction=&Color_Time_of_Day}.Result) {materialFunction=&MF_FX_EmissiveAdjustments}
  n40 = Multiply(A=Time()) {constB=0.5}
  n41 = Panner(Coordinate=Multiply(A=TextureCoordinate() {uTiling=2}, B=Constant() {desc=Macro Iteration, r=2}) {constB=2}) {speedX=-6}
  n42 = MaterialFunctionCall(Float2=Ceil(None=@n41)) {materialFunction=&BreakOutFloat2Components}
  n43 = Multiply(A=@n42.G) {constB=2}
  n44 = Multiply(A=Subtract(A=Add(A=@n42.R, B=@n43)) {constB=3}) {constB=0.3333}
  n45 = Constant() {desc=Micro Iteration, r=4}
  n46 = Ceil(None=Multiply(A=@n41, B=@n45) {constB=4})
  n47 = SphereMask(A=@n46, B=Add(A=@n45) {constB=0.5}, Radius=Add(A=@n45)) {attenuationRadius=5, hardnessPercent=0}
  n48 = ScalarParameter() {defaultValue=1, group=Patterns, parameterName=Color Animation Cycle Time}
  n49 = Constant() {r=2}
  n50 = MaterialFunctionCall(f2=@n46) {materialFunction=&AddComponents}
  n51 = Add(A=Time(), B=Divide(A=Fmod(A=@n50.f2, B=@n49), B=@n49))
  n52 = Multiply(A=Frac(None=Multiply(A=@n48, B=@n51) {constB=0.05, desc=Full cycle 20 second}) {desc="0-1 cycle "}) {constB=3, desc=the number of stages}
  n53 = ScalarParameter() {defaultValue=1, group=Patterns, parameterName=Pattern Transition Speed}
  n54 = MaterialFunctionCall(In=Clamp(None=@n52), Contrast=@n53) {materialFunction=&CheapContrast}
  n55 = Subtract(A=@n52)
  n56 = Clamp(None=@n55)
  n57 = Subtract(A=@n55)
  n58 = LinearInterpolate(
   A=LinearInterpolate(
    A=LinearInterpolate(A=@n44, B=@n47, Alpha=@n54.Result)
    B=Multiply(A=@n50.f2) {constB=0.125}
    Alpha=MaterialFunctionCall(In=@n56, Contrast=@n53) {materialFunction=&CheapContrast}.Result
   )
   B=@n44
   Alpha=MaterialFunctionCall(In=Clamp(None=@n57), Contrast=@n53) {materialFunction=&CheapContrast}.Result
  )
  n59 = Add(A=Sine(None=Add(A=@n40, B=@n58)), B=ScalarParameter() {defaultValue=1, group=Pulse, parameterName=Animation Sine Pulse Low Value}) {constB=1.25}
  n60 = ScalarParameter() {defaultValue=0.1357, group=Pulse, parameterName=Animation Sine Pulse High Value}
  n61 = MaterialFunctionCall(Float2=Ceil(None=@n46)) {materialFunction=&BreakOutFloat2Components}
  n62 = Add(A=@n61.R, B=Multiply(A=@n61.G) {constB=2})
  n63 = Add(
   A=Time()
   B=Multiply(A=Subtract(A=@n62) {constB=3}, B=ScalarParameter() {defaultValue=2.6885, group=Pulse, parameterName=The Magic Number}) {constB=0.05}
  )
  n64 = ScalarParameter() {defaultValue=0.25, group=Pulse, parameterName=Sine Pulse Speed}
  n65 = Add(A=Sine(None=Multiply(A=@n63, B=@n64)), B=ScalarParameter() {defaultValue=5.4534, group=Pulse, parameterName=Sine Pulse Low Value})
  n66 = ScalarParameter() {defaultValue=0.0595, group=Pulse, parameterName=Sine Pulse High Value}
  Multiply(A=Multiply(A=Multiply(A=@n59, B=@n60) {constB=0.5}, B=Multiply(A=@n65, B=@n66)), B=@n28)
  n67 = TextureCoordinate()
  Panner(Coordinate=AppendVector(A=Multiply(), B=ComponentMask(None=@n67) {g=true})) {speedX=1.5}
  ScalarParameter() {defaultValue=0.15, parameterName="Speed (S)"}
  ScalarParameter() {defaultValue=0.15, parameterName="Speed (S)"}
  n68 = Multiply(A=@n34) {constB=3}
  n69 = Constant() {r=0.5}
  n70 = TextureCoordinate()
  n71 = SphereMask(
   A=AppendVector(
    A=@n69
    B=MaterialFunctionCall(Float2=@n70) {materialFunction=&BreakOutFloat2Components}.G
   )
   B=@n69
   Radius=ScalarParameter() {defaultValue=0.5, parameterName=Mask Radius}
   Hardness=ScalarParameter() {defaultValue=0.9, parameterName=Mask Hardness}
  )
  n72 = TextureCoordinate() {uTiling=0.5}
  n73 = Constant2Vector() {g=0.82, r=0.85}
  n74 = Constant() {r=0.8}
  n75 = Constant()
  n76 = Power(Base=OneMinus(None=SphereMask(A=@n72, B=@n73, Radius=@n74, Hardness=@n75))) {constExponent=5}
  n77 = Multiply(
   A=Multiply(A=Saturate(None=@n68), B=Multiply(A=@n71, B=@n76))
   B=MaterialFunctionCall(Distance=ScalarParameter() {defaultValue=64, parameterName=NearFadeDistance}) {materialFunction=&Near_Camera_Fade}.Result
  )
  n78 = MaterialFunctionCall(
   Brightness=ScalarParameter() {defaultValue=1.5, parameterName=Fresnel Brightness}
   Exponent=ScalarParameter() {defaultValue=3, parameterName=Fresnel Exponent}
  ) {materialFunction=&MF_Fresnel_FX}
  n79 = ComponentMask(None=TextureCoordinate()) {r=true}
  n80 = OneMinus(None=@n79)
  n81 = Multiply(A=Multiply(A=Multiply(A=@n38.A, B=@n29), B=Multiply(A=@n77, B=@n78.Fresnel Double Sided)), B=Multiply(A=Multiply(A=@n79, B=@n80)) {constB=2})
  ShadingPathSwitch(Default=DepthFade(Opacity=@n81, FadeDistance=ScalarParameter() {defaultValue=500, parameterName=FadeDistance}), Mobile=@n81) {desc=Migrated from FeatureLevelSwitch}
 ]
}
